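# Unlit, Nanite-compatible material for clumps of bats rendered as Niagara mesh particles.
# Particle color and a texture drive emissive, with dynamic random shade and optional view-angle shading.
# World position offset uses Pivot Painter pivots and axes to flap each bat's two sides with time-based sines, masked along U.
# Dynamic parameters control per-bat scale, spread and grow-out, using vectors and random values stored in extra UV channels, with an optional offset mask.
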Material /DealershipSharedLibrary/Assets/Sets/Bats/M_DLR_BatsClump
graph {
 properties {
  bUsedWithNanite=true
  bUsedWithNiagaraMeshParticles=true
  previewMesh=/DealershipSharedLibrary/Assets/Sets/B2000/SM_DLR_B2000_IsolateBat
  shadingModel=MSM_Unlit
  usedShadingModels=Unlit
 }
 nodes [
  n1 = ParticleColor()
  n2 = ScalarParameter() {defaultValue=32, parameterName=ColorIntensity}
  n3 = TextureSampleParameter2D() {parameterName=MainTexture, texture=&/T_Eyes_A}
  n4 = Add(
   A=Multiply(A=Multiply(A=@n1.RGB, B=@n2) {constB=32}, B=@n3.RGB)
   B=DynamicParameter() {paramNames [Scale,spread,GrowOut,RandomShade]}.RandomShade
  )
  n5 = CameraVectorWS()
  MaterialExpressionConstantClamp_13 = Saturate(None=DotProduct(
    A=Transform(None=@n5) {transformSourceType=TRANSFORMSOURCE_World, transformType=TRANSFORM_Tangent}
    B=Constant3Vector() {constant=<rgba 0 0 1 0>}
   ))
  MaterialFunctionCall(Emissive=LinearInterpolate(A=@n4, B=Multiply(A=@n4, B=@MaterialExpressionConstantClamp_13), Alpha=ScalarParameter() {parameterName=UseCameraVectorInColor?})) {materialFunction=&MF_FX_EmissiveAdjustments}
  Multiply(A=@n1.A, B=@n3.A)
  n6 = WorldPosition()
  n7 = Multiply(A=NamedRerouteUsage(), B=NamedRerouteUsage())
  NamedRerouteDeclaration(None=Add(A=Add(A=@n6.XYZ, B=@n7), B=NamedRerouteUsage())) {name=Pos, nodeColor=<rgba 0.4297 0 1 1>}
  n8 = Add(A=Time(), B=ComponentMask(None=TextureCoordinate() {coordinateIndex=3}) {g=true})
  n9 = Sine(None=@n8)
  n10 = MaterialFunctionCall(X=@n9, Y=Cosine(None=@n8) {period=0.5}, Z=@n9) {materialFunction=&MakeFloat3}
  n11 = NamedRerouteUsage()
  NamedRerouteDeclaration(None=Multiply(A=Multiply(A=@n10.Result, B=@n11), B=ScalarParameter() {parameterName=Random})) {name=Random, nodeColor=<rgba 1 0 0.0309 1>}
  n12 = MaterialFunctionCall() {materialFunction=/Engine/Functions/Engine_MaterialFunctions02/PivotPainter/PivotPainter_PerObjectData}
  n13 = Multiply(A=Add(A=@n12.Random Value Per Element, B=Time()), B=Add(A=ScalarParameter() {defaultValue=1, parameterName=Speed}, B=ParticleRandom()))
  n14 = Sine(None=OneMinus(None=@n13))
  n15 = ComponentMask(None=TextureCoordinate()) {r=true}
  n16 = Constant() {r=3}
  n17 = Saturate(None=Power(Base=@n15, Exp=@n16))
  n18 = ScalarParameter() {defaultValue=0.25, parameterName=Intensity}
  n19 = RotateAboutAxis(
   NormalizedRotationAxis=NamedRerouteUsage()
   RotationAngle=Multiply(A=Multiply(A=@n14, B=@n17), B=@n18)
   PivotPoint=NamedRerouteUsage()
   Position=NamedRerouteUsage()
  )
  n20 = Sine(None=@n13)
  n21 = Saturate(None=Power(Base=OneMinus(None=@n15), Exp=@n16))
  n22 = RotateAboutAxis(
   NormalizedRotationAxis=NamedRerouteUsage()
   RotationAngle=Multiply(A=@n18, B=Multiply(A=@n20, B=@n21))
   PivotPoint=NamedRerouteUsage()
   Position=NamedRerouteUsage()
  )
  n23 = Add(A=Add(A=Add(A=@n19, B=@n22) {constB=0}, B=NamedRerouteUsage()), B=NamedRerouteUsage())
  n24 = NamedRerouteUsage()
  n25 = VectorParameter() {parameterName=Axis_Selector, channelNames {a="", b=Z, g=Y, r=X}, defaultValue=<rgba 1 1 1 0>}
  n26 = Multiply(A=Add(A=Add(A=@n23, B=@n24), B=NamedRerouteUsage()), B=@n25.RGB)
  n27 = Sine(None=ComponentMask(None=TextureCoordinate()) {r=true}) {period=2}
  n28 = ScalarParameter() {
   defaultValue=1
   enumeration=/DealershipSharedLibrary/FX/Materials/Enums/Enum_DLR_TrueFalse
   group=WPO
   parameterName=Contrast_MaskWPO
   sortPriority=1
  }
  n29 = Multiply(
   A=Power(Base=OneMinus(None=@n27), Exp=@n28)
   B=ScalarParameter() {defaultValue=1, group=WPO, parameterName=Brightness_MaskWPO, sortPriority=2}
  )
  n30 = ScalarParameter() {
   controlType=Enumeration
   enumeration=&Enum_DLR_TrueFalse
   group=WPO
   parameterName=MaskWPO
   sortPriority=0
  }
  NamedRerouteDeclaration(None=LinearInterpolate(A=@n26, B=Multiply(A=@n26, B=@n29), Alpha=@n30)) {name=WPO, nodeColor=<rgba 0 0.0859 1 1>}
  NamedRerouteDeclaration(None=@n12.Pivot Position) {name=PerObjectPivot, nodeColor=<rgba 0 1 0.1094 1>}
  NamedRerouteDeclaration(None=@n12.X-Axis Vector) {name=PerObjectXAxis, nodeColor=<rgba 1 0 0.2109 1>}
  n31 = DynamicParameter() {paramNames [Scale,spread,GrowOut,RandomShade]}
  NamedRerouteDeclaration(None=@n31.GrowOut) {name=DYNGrouwOut, nodeColor=<rgba 1 0.1024 0 1>}
  NamedRerouteDeclaration(None=@n31.Scale) {name=DYNScale, nodeColor=<rgba 0.2656 0 1 1>}
  n32 = Constant() {r=-1}
  n33 = Constant() {r=6}
  n34 = ComponentMask(None=TextureCoordinate() {coordinateIndex=3}) {g=true}
  n35 = ScalarParameter() {controlType=Enumeration, enumeration=&Enum_DLR_TrueFalse, parameterName=RandomScale}
  n36 = Add(B=LinearInterpolate(B=Multiply(A=@n33, B=@n34), Alpha=Saturate(None=@n35)) {constA=-1}) {constA=1}
  n37 = NamedRerouteUsage()
  n38 = WorldPosition()
  n39 = NamedRerouteUsage()
  NamedRerouteDeclaration(None=Multiply(A=LinearInterpolate(A=@n32, B=@n36, Alpha=@n37), B=Subtract(A=@n38.XYZ, B=@n39))) {name=Scale, nodeColor=<rgba 1 0.0234 0 1>}
  NamedRerouteDeclaration(None=@n7) {name=GowOutXPerObject, nodeColor=<rgba 1 0 0.5391 1>}
  NamedRerouteDeclaration(None=@n31.spread) {name=DYNSpread, nodeColor=<rgba 0 1 0.5781 1>}
  n40 = TextureCoordinate() {coordinateIndex=2}
  n41 = MaterialFunctionCall(
   X=ComponentMask(None=@n40) {r=true}
   Y=ComponentMask(None=@n40) {g=true}
   Z=ComponentMask(None=TextureCoordinate() {coordinateIndex=3}) {r=true}
  ) {materialFunction=&MakeFloat3}
  NamedRerouteDeclaration(None=Multiply(A=Transform(None=@n41.Result) {transformSourceType=TRANSFORMSOURCE_Local}, B=NamedRerouteUsage())) {name=Spread, nodeColor=<rgba 1 0.2578 0 1>}
  NamedRerouteUsage()
 ]
}
